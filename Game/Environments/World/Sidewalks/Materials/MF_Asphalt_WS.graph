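# Material function that builds a world-aligned asphalt surface as material attributes, with exposed parameters for textures, scale, tint and normal strength.
# Z-projected diffuse, normal and mask textures feed the base layer: mask G drives Metallic and Specular, remapped B drives Roughness, halved at night through Color_Time_of_Day.
# A half-size detail normal is added.
# World-XY noise varies tint and roughness, and is also output as mask and colors.
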MaterialFunction /Game/Environments/World/Sidewalks/Materials/MF_Asphalt_WS
graph {nodes [
  n1 = TextureObjectParameter() {
   group=Asphalt
   parameterName=Asphalt Specular
   samplerType=SAMPLERTYPE_Masks
   texture=/Game/Environments/Sets/Neo/Architecture/Textures/T_Neo_Asphalt1_S
  }
  n2 = ScalarParameter() {defaultValue=328, group=Asphalt, parameterName=Asphalt UVScale, sortPriority=1}
  n3 = MaterialFunctionCall(Float3=MaterialFunctionCall(TextureObject=@n1, TextureSize=@n2) {materialFunction=&WorldAlignedTexture}.Z Texture) {materialFunction=&BreakOutFloat3Components}
  n4 = ScalarParameter() {defaultValue=-0.5714, group=Asphalt, parameterName=Asphalt R Min, sortPriority=4}
  n5 = ScalarParameter() {defaultValue=1.7609, group=Asphalt, parameterName=Asphalt R Max, sortPriority=4}
  n6 = Constant() {r=1}
  n7 = Constant() {r=0.5}
  n8 = TextureObjectParameter() {
   group=Asphalt
   parameterName=Asphalt Normal
   samplerType=SAMPLERTYPE_Normal
   texture=/Game/Environments/Sets/Neo/Architecture/Textures/T_Neo_Asphalt1_N
  }
  n9 = ScalarParameter() {defaultValue=1, group=Asphalt, parameterName=Asphalt Normal Value, sortPriority=5}
  n10 = AppendVector(A=@n9, B=@n9)
  n11 = Constant() {r=1}
  n12 = SetMaterialAttributes!(
   MaterialAttributes
   BaseColor=MaterialFunctionCall(
    TextureObject=TextureObjectParameter() {group=Asphalt, parameterName=Asphalt Diffuse, texture=/Game/Environments/Sets/Neo/Architecture/Textures/T_Neo_Asphalt1_D}
    TextureSize=@n2
   ) {materialFunction=&WorldAlignedTexture}.Z Texture
   Metallic=@n3.G
   Specular=@n3.G
   Roughness=Multiply(
    A=LinearInterpolate(A=@n4, B=@n5, Alpha=@n3.B)
    B=MaterialFunctionCall(Color_Morning=@n6, Color_Day=@n6, Color_Evening=@n6, Color_Night=@n7) {materialFunction=&Color_Time_of_Day}.Result
   )
   Normal=Multiply(
    A=MaterialFunctionCall(TextureObject=@n8, TextureSize=@n2) {materialFunction=&WorldAlignedNormal}.Z Texture
    B=AppendVector(A=@n10, B=@n11)
   )
  ) {attributeSetTypes [
    #1
    #2
    #3
    #4
    #5
   ]}
  n13 = GetMaterialAttributes(None=@n12) {attributeGetTypes [#5]}
  n14 = Multiply(
   A=MaterialFunctionCall(TextureObject=@n8, TextureSize=Multiply(A=@n2) {constB=0.5}) {materialFunction=&WorldAlignedNormal}.Z Texture
   B=VectorParameter() {group=Asphalt, parameterName=Asphalt Details N Value, sortPriority=6, defaultValue=<rgba 2 2 0 1>}.RGB
  )
  n15 = GetMaterialAttributes(None=SetMaterialAttributes!(MaterialAttributes=@n13.MaterialAttributes, Normal=Add(A=@n13.Normal, B=@n14)) {attributeSetTypes [#5]}) {attributeGetTypes [#4, #1]}
  n16 = LinearInterpolate(A=Constant() {r=0.6}, B=Constant() {r=0.7}, Alpha=@n15.Roughness)
  n17 = WorldPosition()
  n18 = TextureSample(UVs=Divide(
    A=ComponentMask(None=@n17.XYZ) {g=true, r=true}
    B=ScalarParameter() {defaultValue=1500, group=Asphalt, parameterName=Asphalt Noise Scale, sortPriority=3}
   )) {samplerType=SAMPLERTYPE_LinearColor, texture=/Game/Global/Textures/Default/Noise/T_Noise01_Inverse}
  n19 = VectorParameter() {group=Asphalt, parameterName=Asphalt Tint A, sortPriority=2, defaultValue=<rgba 0.9948 0.9948 0.9948 0>}
  n20 = VectorParameter() {desc="Green ", group=Asphalt, parameterName=Asphalt Tint B, sortPriority=2, defaultValue=<rgba 0.8021 0.7611 0.6833 0>}
  n21 = LinearInterpolate(A=@n19.RGB, B=@n20.RGB, Alpha=@n18.R)
  FunctionOutput(None=SetMaterialAttributes!(
    MaterialAttributes=@n15.MaterialAttributes
    Roughness=LinearInterpolate(A=@n15.Roughness, B=@n16, Alpha=@n18.R)
    BaseColor=Multiply(A=@n15.BaseColor, B=@n21)
   ) {attributeSetTypes [#4, #1]})
  FunctionOutput(None=@n18.R) {outputName=Noise Mask, sortPriority=2}
  FunctionOutput(None=@n21) {outputName=Noise Colors, sortPriority=2}
 ]}
