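# M_GroundPaintFactory is an asphalt ground material with optional painted markings, landscape blending and puddles.
# A world-space asphalt function feeds a switchable paint layer, masked by a selectable texture channel with adjustable UVs.
# Vertex alpha height-blends color, roughness and specular toward the Apollo landscape runtime virtual texture, behind an RVT switch.
# Vertex blue and world-space noise then blend in puddles with low roughness, flattened normals and Fresnel-driven tint.
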Material /Game/Environments/World/Sidewalks/Materials/M_GroundPaintFactory
graph {
 properties {bUseMaterialAttributes=true, pixelDepthOffsetMode=PDOM_Legacy, usedShadingModels=Default Lit}
 nodes [
  n1 = MaterialFunctionCall() {materialFunction=&/MF_Asphalt_WS}
  n2 = GetMaterialAttributes(None=@n1.Result) {attributeGetTypes [#1, #2, #3]}
  n3 = VectorParameter() {group=PaintLayer, parameterName=PaintColor, defaultValue=<rgba 0.7 0.7 0.7 0>}
  n4 = TextureCoordinate()
  n5 = VectorParameter() {group=PaintLayer, parameterName=PaintLayerUV_RG_Offset_BA_Scale, defaultValue=<rgba 0 0 1 1>}
  n6 = ComponentMask(None=@n5.RGB) {g=true, r=true}
  n7 = TextureSampleParameter2D(UVs=Multiply(A=Add(A=@n4, B=@n6), B=AppendVector(A=@n5.B, B=@n5.B))) {
   group=PaintLayer
   parameterName=PaintLayerMasks
   samplerType=SAMPLERTYPE_Masks
   texture=/Game/Packages/DS_Fortnight/SM/Textures/T_SoccerTennisBasketball
  }
  n8 = StaticComponentMaskParameter(None=AppendVector(A=@n7.RGB, B=@n7.RGB)) {defaultR=true, group=PaintLayer, parameterName=PaintLayerMaskChannel}
  n9 = Multiply(A=@n2.Specular, B=Constant() {r=1.25})
  n10 = StaticSwitchParameter(
   True=SetMaterialAttributes!(
    MaterialAttributes=@n2.MaterialAttributes
    Base Color=LinearInterpolate(A=@n2.Base Color, B=@n3.RGB, Alpha=@n8)
    Specular=LinearInterpolate(A=@n2.Specular, B=@n9, Alpha=@n8)
    Roughness=LinearInterpolate(A=@n2.Roughness, Alpha=@n8) {constB=0.4}
   ) {attributeSetTypes [#1, #2, #3]}
   False=@n1.Result
  ) {desc=Diffuse, group=Layers, parameterName=Use PaintLayer}
  n11 = GetMaterialAttributes(None=@n10) {attributeGetTypes [#1, #3, #2]}
  n12 = GetMaterialAttributes(None=@n1.Result) {attributeGetTypes [#3]}
  n13 = RuntimeVirtualTextureSample!(UVs, World Position, Mip Level) {
   materialType=BaseColor_Normal_Specular_YCoCg
   mipValueMode=RVTMVM_MipLevel
   virtualTexture=/Game/Athena/Apollo/Maps/Landscape/VT/RVT_Apollo_Landscape
  }
  n14 = MaterialFunctionCall(
   A=@n11.Base Color
   B=@n13.BaseColor
   Transition Phase=VertexColor().A
   Height Texture=Subtract(A=@n1.Noise Mask, B=ComponentMask(None=@n12.Roughness) {r=true})
   Contrast=Constant() {r=5}
  ) {materialFunction=&HeightLerp}
  n15 = MaterialFunctionCall(
   NoRVT=@n10
   RVT=SetMaterialAttributes!(
    MaterialAttributes=@n11.MaterialAttributes
    Base Color=@n14.Results
    Roughness=LinearInterpolate(A=@n11.Roughness, B=@n13.Roughness, Alpha=@n14.Alpha)
    Specular=LinearInterpolate(A=@n11.Specular, B=@n13.Specular, Alpha=@n14.Alpha)
   ) {attributeSetTypes [#1, #3, #2]}
  ) {materialFunction=&FN_RVT_Switch_Material}
  n16 = GetMaterialAttributes(None=@n15.Result) {attributeGetTypes [
    #3
    #2
    #1
    #4
   ]}
  n17 = Constant() {r=-0.2}
  n18 = Constant() {r=0.8}
  n19 = ComponentMask(None=@n16.Roughness) {r=true}
  n20 = ComponentMask(None=WorldPosition().XYZ) {g=true, r=true}
  n21 = ScalarParameter() {defaultValue=355, group=Puddles, parameterName=Puddles Scale}
  n22 = MaterialFunctionCall(
   Transition Phase=VertexColor().B
   Height Texture=TextureSample(UVs=Divide(A=@n20, B=@n21)) {samplerType=SAMPLERTYPE_LinearColor, texture=/Game/Global/Textures/Default/Noise/T_Noise01_Inverse}.R
   Contrast=ScalarParameter() {defaultValue=1, group=Puddles, parameterName=PuddlesContrast}
  ) {materialFunction=&HeightLerp}
  n23 = Constant() {r=0.01}
  n24 = Constant() {r=0.1}
  n25 = Fresnel() {baseReflectFraction=0.125, exponent=3}
  n26 = LinearInterpolate(B=VectorParameter() {group=Puddles, parameterName=Puddles Color, defaultValue=<rgba 0.2626 0.071 1 0>}.RGB, Alpha=@n25) {constA=0.85}
  n27 = SetMaterialAttributes!(
   MaterialAttributes=@n16.MaterialAttributes
   Roughness=LinearInterpolate(A=LinearInterpolate(A=@n17, B=@n18, Alpha=@n19), B=Constant() {r=0.1}, Alpha=@n22.Alpha)
   Specular=LinearInterpolate(A=Multiply(A=@n16.Specular) {constB=0.5}, B=LinearInterpolate(A=@n23, B=@n24, Alpha=@n25), Alpha=@n22.Alpha)
   Base Color=LinearInterpolate(A=Multiply(A=@n16.Base Color) {constB=0.85}, B=Multiply(A=@n16.Base Color, B=@n26), Alpha=@n22.Alpha)
   Normal=LinearInterpolate(A=@n16.Normal, B=Constant3Vector() {constant=<rgba 0 0 1 0>}, Alpha=@n22.Alpha)
  ) {attributeSetTypes [
    #3
    #2
    #1
    #4
   ]}
  MaterialFunctionCall!(Base Material=@n15.Result, Top Material=@n27, Alpha=@n22.Lerp Alpha No Contrast) {materialFunction=/Engine/Functions/MaterialLayerFunctions/MatLayerBlend_Standard}
 ]
}
