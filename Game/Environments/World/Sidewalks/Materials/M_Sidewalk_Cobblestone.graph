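# Sidewalk material for Nanite and instanced meshes that blends world-aligned cobblestones into a UV-mapped background using the background diffuse's alpha.
# World-space noise varies cobblestone tint and normal flattening.
# Puddles use HeightLerp of vertex blue, noise and stone height (diffuse alpha) to alter roughness, specular, color and normal, with Fresnel shading.
# A static switch can substitute a generic crosswalk-textured surface, and either result feeds the Apollo snow progression function.
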Material /Game/Environments/World/Sidewalks/Materials/M_Sidewalk_Cobblestone
graph {
 properties {
  bUseMaterialAttributes=true
  bUsedWithInstancedStaticMeshes=true
  bUsedWithNanite=true
  pixelDepthOffsetMode=PDOM_Legacy
  usedShadingModels=Default Lit
 }
 nodes [
  n1 = TextureSample() {texture=/Game/Environments/Sets/Neo/Architecture/Textures/T_Neo_Crosswalk1_D}
  n2 = TextureSample() {samplerType=SAMPLERTYPE_Normal, texture=/Game/Environments/Sets/Neo/Architecture/Textures/T_Neo_Crosswalk1_N}
  n3 = Constant3Vector() {constant=<rgba 0 0 1 0>}
  n4 = SetMaterialAttributes!(
   MaterialAttributes
   Roughness=Multiply(
    A=TextureSample() {samplerType=SAMPLERTYPE_Masks, texture=/Game/Environments/Apollo/Sets/PowerPlant/Textures/T_PP_Concrete_Wall_1_S}.B
    B=ScalarParameter() {defaultValue=0.7, parameterName=Generic-RoughnessMod}
   )
   Specular=Constant() {r=0.5}
   Base Color=@n1.RGB
   Normal=ShadingPathSwitch(Default=@n2.RGB, Deferred=@n2.RGB, Forward=@n3, Mobile=@n3)
   Opacity Mask=@n1.RGB
  ) {attributeSetTypes [
    #1
    #2
    #3
    #4
    #5
   ]}
  n5 = TextureSampleParameter2D() {group=MeshTextures, parameterName=Background D, texture=/Game/Packages/DS_Fortnight/SM/Textures/T_Sidewalk_3}
  n6 = TextureSampleParameter2D() {
   group=MeshTextures
   parameterName=Background S
   samplerType=SAMPLERTYPE_Masks
   texture=/Game/Packages/DS_Fortnight/SM/Textures/T_Sidewalk_1_M
  }
  n7 = TextureSampleParameter2D() {
   group=MeshTextures
   parameterName=Background N
   samplerType=SAMPLERTYPE_Normal
   texture=/Game/Packages/DS_Fortnight/SM/Textures/T_Sidewalk_1_N
  }
  n8 = TextureObjectParameter() {group=Tile Textures, parameterName=Diffuse, sortPriority=1, texture=/Game/Environments/Apollo/Textures/T_Apollo_Cobblestone_D_01}
  n9 = TextureObjectParameter() {
   group=Tile Textures
   parameterName=SpecularMasks
   samplerType=SAMPLERTYPE_Masks
   sortPriority=2
   texture=/Game/Environments/Apollo/Textures/T_Apollo_Cobblestone_S_01
  }
  n10 = TextureObjectParameter() {
   group=Tile Textures
   parameterName=Normals
   samplerType=SAMPLERTYPE_Normal
   sortPriority=3
   texture=/Game/Environments/Apollo/Textures/T_Apollo_Cobblestone_N_01
  }
  n11 = MaterialFunctionCall!(
   Diffuse Texture Object=@n8
   Spec Texture Object=@n9
   TextureSize=ScalarParameter() {defaultValue=256, group=Tile Textures, parameterName=Tile Size}
   --------------------------
   Use seperate settings for Z projection
   Use seperate texture for Z projection
   Texture Size Z projection
   Z Plane Diffuse Texture Object
   Z Plane Spec Texture Object
   "--------------------------      "
   ProjectionTransitionContrast
   Use Texture Alpha Heightmap to Refine the Transistion
   HeightLerpTransitionContrast
   PickTheDiffuseHeightChannel
   TheDiffuseHeightChannel
   "--------------------------       "
   NormalTextureObject=@n10
   World Space Normals=StaticBool()
   "--------------------------  "
   Export Float 4 Diffuse=StaticBool() {value=true}
   Export Float 4 Spec
   "--------------------------     "
   WorldPosition
   World Space Normal
  ) {materialFunction=/Engine/Functions/Engine_MaterialFunctions02/Texturing/WorldAlignedTextures_Complex}
  n12 = VectorParameter() {group=WS, parameterName=WS Tint A, sortPriority=2, defaultValue=<rgba 0.9948 0.9948 0.9948 0>}
  n13 = VectorParameter() {desc="Green ", group=WS, parameterName=WS Tint B, sortPriority=2, defaultValue=<rgba 0.8021 0.7611 0.6833 0>}
  n14 = WorldPosition()
  n15 = TextureSampleParameter2D(UVs=Divide(
    A=ComponentMask(None=@n14.XYZ) {g=true, r=true}
    B=ScalarParameter() {defaultValue=1500, group=WS, parameterName=WS Noise Scale, sortPriority=3}
   )) {group=WS, parameterName=WS Noise, samplerType=SAMPLERTYPE_LinearColor, texture=/Game/Global/Textures/Default/Noise/T_Noise01_Inverse}
  n16 = MaterialFunctionCall(Float3=@n11.Z Spec) {materialFunction=&BreakOutFloat3Components}
  n17 = ScalarParameter() {defaultValue=0.65, parameterName=FlattenNormal}
  n18 = SetMaterialAttributes!(
   MaterialAttributes
   Base Color=Multiply(A=ComponentMask(None=@n11.Z Diffuse Texture) {b=true, g=true, r=true}, B=LinearInterpolate(A=@n12.RGB, B=@n13.RGB, Alpha=@n15.R))
   Metallic=@n16.G
   Specular=@n16.G
   Roughness=@n16.G
   Normal=MaterialFunctionCall(Normal=@n11.Normal Z Texture, Flatness=LinearInterpolate(B=@n17, Alpha=@n15.R)) {materialFunction=&FlattenNormal}.Result
  ) {attributeSetTypes [
    #3
    #6
    #2
    #1
    #4
   ]}
  n19 = BlendMaterialAttributes(
   A=MakeMaterialAttributes!(
    BaseColor=@n5.RGB
    Metallic=@n6.G
    Specular=@n6.G
    Roughness=@n6.G
    Anisotropy
    EmissiveColor
    Opacity
    OpacityMask
    Normal=@n7.RGB
    Tangent
    WorldPositionOffset
    SubsurfaceColor
    ClearCoat
    ClearCoatRoughness
    AmbientOcclusion
    Refraction
    CustomizedUVs_0
    CustomizedUVs_1
    CustomizedUVs_2
    CustomizedUVs_3
    CustomizedUVs_4
    CustomizedUVs_5
    CustomizedUVs_6
    CustomizedUVs_7
    PixelDepthOffset
    ShadingModel
    Displacement
   )
   B=@n18
   Alpha=@n5.A
  )
  n20 = GetMaterialAttributes(None=@n19) {attributeGetTypes [
    #1
    #2
    #3
    #4
   ]}
  n21 = Constant() {r=-0.2}
  n22 = Constant() {r=0.8}
  n23 = ComponentMask(None=@n20.Roughness) {r=true}
  n24 = WorldPosition()
  n25 = TextureSampleParameter2D(UVs=Divide(A=ComponentMask(None=@n24.XYZ) {g=true, r=true}, B=ScalarParameter() {defaultValue=355, group=Puddles, parameterName=Puddles Scale})) {group=WS, parameterName=WS Noise, samplerType=SAMPLERTYPE_LinearColor, texture=&T_Noise01_Inverse}
  n26 = ComponentMask(None=@n11.Z Diffuse Texture) {a=true}
  n27 = MaterialFunctionCall(
   Transition Phase=OneMinus(None=VertexColor().B)
   Height Texture=Min(A=@n25.R, B=OneMinus(None=@n26))
   Contrast=ScalarParameter() {defaultValue=1, group=Puddles, parameterName=PuddlesContrast}
  ) {materialFunction=&HeightLerp}
  n28 = Constant() {r=0.01}
  n29 = Constant() {r=0.1}
  n30 = Fresnel() {baseReflectFraction=0.125, exponent=3}
  n31 = LinearInterpolate(B=VectorParameter() {group=Puddles, parameterName=Puddles Color, defaultValue=<rgba 1 1 1 0>}.RGB, Alpha=@n30) {constA=0.85}
  n32 = SetMaterialAttributes!(
   MaterialAttributes=@n20.MaterialAttributes
   Roughness=LinearInterpolate(A=LinearInterpolate(A=@n21, B=@n22, Alpha=@n23), B=Constant() {r=0.1}, Alpha=@n27.Alpha)
   Specular=LinearInterpolate(A=Multiply(A=@n20.Specular) {constB=0.5}, B=LinearInterpolate(A=@n28, B=@n29, Alpha=@n30), Alpha=@n27.Alpha)
   Base Color=LinearInterpolate(A=Multiply(A=@n20.Base Color) {constB=0.85}, B=Multiply(A=@n20.Base Color, B=@n31), Alpha=@n27.Alpha)
   Normal=LinearInterpolate(A=@n20.Normal, B=Constant3Vector() {constant=<rgba 0 0 1 0>}, Alpha=@n27.Alpha)
  ) {attributeSetTypes [
    #1
    #2
    #3
    #4
   ]}
  n33 = MaterialFunctionCall(In=@n27.Lerp Alpha No Contrast, Contrast=ScalarParameter() {group=Puddles, parameterName=PuddlesWet Contrast}) {materialFunction=&CheapContrast}
  MaterialFunctionCall!(
   IsStaticSnow
   UseAlternateStaticSnowMask
   AlternateStaticSnowMask
   "UseRoad&&OcclusionMask"
   IsForGrassMeshToPushDown
   AdditionalMaskToRemoveSnow
   IsFallingSnowFX
   UseMinimapCoords
   IncreaseSnowUsingSlope=StaticBool()
   UseSmoothBorder
   LerpHighM
   TemporalHeightLerpContrast
   IsLobbyIsland
   UseBR-IslandMask
   IsTree
   IsForRenderTarget
   UseSparkle
   UseSnowNormals
   UseAdditionalMaskToRemoveSnow
   IsTangentSpaceInput=StaticBool() {value=true}
   UseVertexNormalsForBasicSnow
   UseTangentSnowNormals
   Material=StaticSwitchParameter(True=@n4, False=BlendMaterialAttributes(A=@n19, B=@n32, Alpha=@n33.Result)) {parameterName=UseGenericMaterial}
  ) {materialFunction=/Game/Athena/Apollo/Environments/Landscape/Materials/Biome/UniqueBiomeMaterialFunctions/MF_Apollo_Snow_Progression}
  n34 = OneMinus(None=ComponentMask(None=TextureCoordinate()) {r=true})
  n35 = ScalarParameter() {defaultValue=20, parameterName=Mask-Mod}
  Saturate(None=Add(A=Multiply(A=@n34, B=@n35) {constB=2}, B=ScalarParameter() {defaultValue=-2, parameterName=Mask-Add}) {constB=-0.75})
 ]
}
